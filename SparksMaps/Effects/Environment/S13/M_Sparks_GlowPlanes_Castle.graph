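# Translucent unlit material for Sparks castle glow planes, giving an animated, audio-reactive smoky fire glow.
# Two panning wispy noise samples blend into a Fire layer, which is gradient-faded, added to a vertical glow, then masked at sides and bottom.
# The audio collection's BeatLifetime modulates both the fire and the combined mask.
# A separate branch scales the secondary Mojo stage color by a brightness parameter.
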Material /SparksMaps/Effects/Environment/S13/M_Sparks_GlowPlanes_Castle
graph {
 properties {bAutomaticallySetUsageInEditor=false, bHasPixelAnimation=true, blendMode=BLEND_Translucent, shadingModel=MSM_Unlit, usedShadingModels=Unlit}
 nodes [
  Constant() {desc=BeatHigh, r=0.5}
  Constant() {desc=BeatLow, r=1.2}
  Multiply(
   A=MaterialFunctionCall!(CustomUPosition) {materialFunction=/SparksStageShow/Systems/StageManager/FXManager/MaterialFunctions/MF_MojoStageColors}.Secondary
   B=ScalarParameter() {defaultValue=3, parameterName=SmokeBright}
  )
  n1 = ScalarParameter() {defaultValue=1, parameterName=Mask_SIde_Min}
  n2 = ScalarParameter() {defaultValue=0.7, parameterName=MaskSide_Max}
  n3 = ComponentMask(None=TextureCoordinate()) {r=true}
  n4 = Multiply(
   A=SmoothStep(Min=@n1, Max=@n2, Value=@n3) {constMax=0.8, constMin=0.6}
   B=SmoothStep(Min=@n1, Max=@n2, Value=OneMinus(None=@n3)) {constMax=0.8, constMin=0.6}
  )
  n5 = SmoothStep(
   Min=ScalarParameter() {defaultValue=0.1, parameterName=Fire_Gradient_Min}
   Max=ScalarParameter() {defaultValue=1, parameterName=Fire_Gradient_Max}
   Value=ComponentMask(None=TextureCoordinate()) {g=true}
  ) {constMax=0.8, constMin=0.6}
  n6 = ScalarParameter() {defaultValue=0.05, parameterName=SmokeIntensity was .4}
  n7 = ScalarParameter() {defaultValue=0.1, parameterName=Glow_Min}
  n8 = ScalarParameter() {defaultValue=1, parameterName=Glow_Max}
  n9 = ComponentMask(None=TextureCoordinate()) {g=true}
  n10 = Add(
   A=Multiply(A=Multiply(A=@n5, B=@n6), B=NamedRerouteUsage())
   B=Multiply(
    A=SmoothStep(Min=@n7, Max=@n8, Value=@n9) {constMax=0.8, constMin=0.6}
    B=ScalarParameter() {defaultValue=0.3, parameterName=Glow_Power was .5}
   )
  ) {desc=Combines Fire and Glow}
  n11 = SmoothStep(
   Min=ScalarParameter() {defaultValue=1, parameterName=Mask_Bottom_Min}
   Max=ScalarParameter() {defaultValue=0.9, parameterName=Mask_Bottom_Max}
   Value=ComponentMask(None=TextureCoordinate()) {g=true}
  ) {constMax=0.8, constMin=0.6}
  Multiply(
   A=Multiply(A=Multiply(B=@n4) {constA=1}, B=Multiply(A=@n10, B=@n11) {desc=Fire + Glow + Bottom Mask}) {desc=All Masks}
   B=LinearInterpolate(
    A=Constant() {desc=BeatLow, r=1.1}
    B=Constant() {desc=BeatHigh, r=0.9}
    Alpha=CollectionParameter() {collection=/FMAudioReactiveData/AudioAnalysis/MPC_Sparks_Audio_Parameters, parameterName=BeatLifetime}
   ) {desc=BeatData}
  )
  n12 = ScalarParameter() {defaultValue=1.2, parameterName=Scale X}
  n13 = ComponentMask(None=TextureCoordinate()) {g=true}
  n14 = ScalarParameter() {defaultValue=1, parameterName=Flame Top UV Stretch}
  n15 = Add(A=Multiply(A=ScalarParameter() {defaultValue=-0.2, parameterName=Scale Y}, B=Multiply(A=@n13, B=@n14))) {constB=0.5}
  n16 = Panner(
   Coordinate=Multiply(A=TextureCoordinate() {uTiling=2, vTiling=2}, B=AppendVector(A=@n12, B=@n15))
   Time=Time()
   Speed=AppendVector(A=ScalarParameter() {defaultValue=-0.05, parameterName=Speed X}, B=ScalarParameter() {defaultValue=0.15, parameterName=Speed Y})
  )
  n17 = ScalarParameter() {defaultValue=1, parameterName=Fire Tile 1 U}
  n18 = ScalarParameter() {defaultValue=0.3, parameterName=Fire Tile 1 V}
  n19 = Panner(
   Coordinate=Multiply(A=TextureCoordinate(), B=AppendVector(A=@n17, B=@n18))
   Time=Time()
   Speed=AppendVector(A=ScalarParameter() {defaultValue=-0.05, parameterName=Speed X}, B=ScalarParameter() {defaultValue=0.15, parameterName=Speed Y})
  )
  n20 = SmoothStep(
   Min=ScalarParameter() {parameterName=SSMin}
   Max=ScalarParameter() {defaultValue=0.8, parameterName=SSMax}
   Value=LinearInterpolate(
    A=TextureSample(UVs=@n16) {texture=/Game/Effects/Fort_Effects/Textures/T_WispyFireNoise03}.R
    B=TextureSample(UVs=@n19) {texture=&T_WispyFireNoise03}.R
    Alpha=ScalarParameter() {defaultValue=0.5, parameterName=Use Dive Pan Speed}
   )
  ) {constMin=0.1}
  NamedRerouteDeclaration(None=LinearInterpolate(
    A=Multiply(B=@n20) {constA=1, constB=1.2}
    B=Multiply(A=@n20)
    Alpha=CollectionParameter() {collection=&MPC_Sparks_Audio_Parameters, parameterName=BeatLifetime}
   )) {name=Fire, nodeColor=<rgba 0 0.6719 1 1>}
 ]
}
